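; LimeRock gameplay mutator that overrides the damage number info shown when an actor is hit.
; ModifyDamageNumberInfo zeroes one field to keep numbers sharp, passes valid targets to an enemy guard helper, then always returns OverrideReturn, ignoring the helper's result.
; For guards tagged RipeHoneydew.EG, the helper scales critical damage by a random multiplier within a configured range and truncates it.
; Its shield-versus-pawn type selection has a constant false condition, so shield state is never checked.
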(event EventTick
  (params (DeltaSeconds :type Float%20%28single-precision%29)))

(event EventBeginPlay)

(fn ModifyDamageNumberForEnemyGuard
  (params (InOutDamageNumberInfo :type Fort%20Damage%20Number%20Info%20Structure%20%28by%20ref%29) (DamagedActor :type Actor%20Object%20Reference) (DamageCauser :type Pawn%20Object%20Reference) (HitPosition :type Vector))
  (bind (_biscriticaldamage _damage) (Utilities|Struct|BreakFortDamageNumberInfo InOutDamageNumberInfo))
  (bind (_minmaxcritmultiplier_x _minmaxcritmultiplier_y) (Math|Vector2D|BreakVector2D (Variables|GetMinMaxCritMultiplier_X)))
  ;[0.1895 0.07615 0.2031] Update the damage type number based on if the guard is shielded or not.
  (bind _returnvalue (GameplayTags|HasMatchingGameplayTag(Message) DamagedActor "(TagName=\"RipeHoneydew.EG\")"))
  (if _returnvalue
    (Utilities|Struct|SetmembersinFortDamageNumberInfo InOutDamageNumberInfo (Math|Float|Truncate (* _damage (Math|Float|SelectFloat (Math|Random|RandomFloatinRange _minmaxcritmultiplier_x _minmaxcritmultiplier_y) 1.0 _biscriticaldamage))) (select false "Shield" "Pawn"))
    (return "OverrideReturn")
    (else
      (return "Ignore"))))

(fn ModifyDamageNumberInfo
  (params (InOutDamageNumberInfo :type Fort%20Damage%20Number%20Info%20Structure%20%28by%20ref%29) (DamagedActor :type Actor%20Object%20Reference) (DamageCauser :type Pawn%20Object%20Reference) (HitPosition :type Vector%20%28by%20ref%29 :default "0, 0, 0"))
  ;[0.01801 0.224 0.2083] Based on the damaged actor, return the modified damage number info.
  (Utilities|Struct|SetmembersinFortDamageNumberInfo InOutDamageNumberInfo 0.0) ; This makes the numbers not blurry
  (Utilities|FlowControl|Sequence
    (:then_0
      (Utilities|IsValid DamagedActor
        (:Is%20Valid
          ;;[1 1 1] Enemy Guard.
          (bind _retvalue (CallFunction|ModifyDamageNumberforEnemyGuard (Variables|GetInOutDamageNumberInfo) (Variables|GetDamagedActor) (Variables|GetDamageCauser) (Variables|GetHitPosition))))))
    (:then_1
      (return "OverrideReturn"))))

(fn ConstructionScript)
